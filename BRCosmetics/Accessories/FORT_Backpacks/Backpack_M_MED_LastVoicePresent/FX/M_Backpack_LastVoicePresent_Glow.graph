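# Unlit, two-sided masked material for the glow FX on the LastVoicePresent backpack, enabled for Niagara sprite and mesh particles.
# It samples a light mask texture, multiplies it by particle color, and passes the result through MF_FX_EmissiveAdjustments for emissive.
# The mask is also multiplied by particle alpha.
# The "Use Red Channel?" lerp reads green on both inputs, so that parameter has no effect.
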Material /BRCosmetics/Accessories/FORT_Backpacks/Backpack_M_MED_LastVoicePresent/FX/M_Backpack_LastVoicePresent_Glow
graph {
 properties {
  bUsedWithNiagaraMeshParticles=true
  bUsedWithNiagaraSprites=true
  blendMode=BLEND_Masked
  pixelDepthOffsetMode=PDOM_Legacy
  shadingModel=MSM_Unlit
  twoSided=true
  usedShadingModels=Unlit
 }
 nodes [
  n1 = ParticleColor()
  n2 = TextureSample() {texture=&/T_FX_LastVoicePresent_Light_Mask}
  n3 = LinearInterpolate(A=@n2.G, B=@n2.G, Alpha=ScalarParameter() {defaultValue=1, parameterName=Use Red Channel?})
  MaterialFunctionCall(Emissive=Multiply(A=@n1.RGB, B=@n3)) {materialFunction=&MF_FX_EmissiveAdjustments}
  Multiply(A=@n3, B=@n1.A)
 ]
}
