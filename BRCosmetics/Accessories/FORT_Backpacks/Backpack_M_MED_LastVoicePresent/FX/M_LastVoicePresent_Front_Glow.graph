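# Additive, unlit, two-sided Niagara sprite material for a front glow effect on the LastVoicePresent backpack cosmetic.
# Particle color RGB runs through a shared emissive adjustment function.
# Particle alpha multiplies a glow shape, blended by parameter between a procedural radial SphereMask and a mask texture, defaulting to the texture.
# That result is multiplied by a view-angle fade and an optional smoothstep gradient along UV V.
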Material /BRCosmetics/Accessories/FORT_Backpacks/Backpack_M_MED_LastVoicePresent/FX/M_LastVoicePresent_Front_Glow
graph {
 properties {
  bUsedWithNiagaraSprites=true
  blendMode=BLEND_Additive
  pixelDepthOffsetMode=PDOM_Legacy
  shadingModel=MSM_Unlit
  twoSided=true
  usedShadingModels=Unlit
 }
 nodes [
  n1 = ParticleColor()
  MaterialFunctionCall(Emissive=@n1.RGB) {materialFunction=&MF_FX_EmissiveAdjustments}
  n2 = TextureCoordinate()
  n3 = Constant() {r=0.5}
  n4 = ScalarParameter() {defaultValue=0.5, parameterName=Radius}
  n5 = ScalarParameter() {parameterName=Hardness}
  n6 = LinearInterpolate(
   A=Power(Base=SphereMask(A=@n2, B=@n3, Radius=@n4, Hardness=@n5) {hardnessPercent=0}, Exp=ScalarParameter() {defaultValue=2, parameterName=Power}) {constExponent=3}
   B=TextureSample() {texture=&/T_FX_LastVoicePresent_Glow_Mask}.R
   Alpha=ScalarParameter() {defaultValue=1, parameterName=Use Texture?}
  )
  n7 = ScalarParameter() {defaultValue=6, parameterName=AngleFade}
  n8 = ScalarParameter() {defaultValue=0.4, parameterName=Value}
  n9 = Add(A=ScalarParameter() {defaultValue=0.2, parameterName=Soft}, B=@n8)
  n10 = MaterialFunctionCall(Float2=TextureCoordinate()) {materialFunction=&BreakOutFloat2Components}
  Multiply(
   A=Multiply(
    A=MaterialFunctionCall("FadeSensitivity (optional)"=@n7) {materialFunction=&FadeBasedOnViewAngle}.Result
    B=Multiply(A=@n1.A, B=@n6)
   )
   B=LinearInterpolate(B=SmoothStep(Min=@n8, Max=@n9, Value=@n10.G), Alpha=ScalarParameter() {defaultValue=1, parameterName=Use Mask?}) {constA=1}
  )
 ]
}
