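# Additive unlit material for Niagara mesh particles, drawing animated rays for the LastVoicePresent backpack effect.
# Two panned plasma noise samples with separate tiling are summed, with a dynamic parameter scaling scroll time.
# That sum blends between rounded and raw values, adds a V gradient, and drives a lerp from dark color to particle color.
# The gradient times inverted Fresnel forms a mask that fades silhouette edges.
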Material /BRCosmetics/Accessories/FORT_Backpacks/Backpack_M_MED_LastVoicePresent/FX/M_LastVoicePresent_Rays
graph {
 properties {
  bUsedWithNiagaraMeshParticles=true
  blendMode=BLEND_Additive
  pixelDepthOffsetMode=PDOM_Legacy
  shadingModel=MSM_Unlit
  usedShadingModels=Unlit
 }
 nodes [
  n1 = TextureCoordinate()
  n2 = Multiply(
   A=AppendVector(A=ScalarParameter() {defaultValue=2, parameterName=U Repeat A}, B=ScalarParameter() {defaultValue=0.0196, parameterName=V Repeat A})
   B=@n1
  )
  n3 = Multiply(
   A=Multiply(A=Time(), B=ScalarParameter() {defaultValue=0.1, parameterName=Time Mult})
   B=DynamicParameter() {paramNames [Time Random, Param2, Param3, Param4]}.Time Random
  )
  n4 = Multiply(
   A=@n1
   B=AppendVector(A=ScalarParameter() {defaultValue=1, parameterName=U Repeat B}, B=ScalarParameter() {defaultValue=0.0117, parameterName=V Repeat B})
  )
  n5 = Add(
   A=TextureSample(UVs=Panner(Coordinate=@n2, Time=@n3) {bFractionalPart=true, speedX=0.031, speedY=0.027}) {texture=/Game/Effects/Fort_Effects/Textures/Plasma/T_PlasmaNoise_A_01}.B
   B=TextureSample(UVs=Panner(Coordinate=@n4, Time=@n3) {bFractionalPart=true, speedX=-0.1485, speedY=0.0218}) {texture=&T_PlasmaNoise_A_01}.R
  )
  n6 = MaterialFunctionCall!(In=@n5) {materialFunction=/Engine/Functions/Engine_MaterialFunctions02/Utility/Round}
  n7 = ScalarParameter() {defaultValue=0.6, parameterName=Edge LERP}
  n8 = Power(Base=ComponentMask(None=TextureCoordinate()) {g=true}, Exp=ScalarParameter() {defaultValue=1.5, parameterName=GradientPower})
  LinearInterpolate(
   A=VectorParameter() {parameterName=DarkColor, defaultValue=<rgba 0.1396 0.0965 0.5 0>}.RGB
   B=ParticleColor().RGB
   Alpha=Add(A=LinearInterpolate(A=@n6.Result, B=@n5, Alpha=@n7), B=@n8)
  )
  n9 = Fresnel(ExponentIn=ScalarParameter() {defaultValue=0.2, parameterName=Exponent}) {baseReflectFraction=0}
  Multiply(A=@n8, B=Multiply(A=OneMinus(None=@n9), B=ScalarParameter() {defaultValue=1, parameterName=Fresnel Multiplier}))
 ]
}
