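# Idle Niagara effect for the LastVoicePresent backpack cosmetic, a persistent warm glow built from burst-spawned particles that never expire.
# Beams and Glows each place eight sprites across two sockets of User.SK_Mesh by spawn group, with beam lengths pulsing in a staggered loop.
# Front_Glow adds a socket-attached sprite with an air-glow layer, while InnerLights and Glow_Rays render meshes.
# Every emitter scales color through time-of-day and front-end modifiers.
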NiagaraSystem /BRCosmetics/Accessories/FORT_Backpacks/Backpack_M_MED_LastVoicePresent/FX/NS_Backpack_LastVoicePresent_Idle
dependencies {
 usedDataInterfaces [&NiagaraDataInterfaceArrayFloat2, &NiagaraDataInterfaceArrayFloat3, &NiagaraDataInterfaceCurve]
 usedDynamicInputs []
 usedModules [&SystemState]
}
emitters {
 Beams {
  emitterUpdateScript [
   &EmitterState(
    Inactive Response=::NewEnumerator1
    Life Cycle Mode=::NewEnumerator1
    Loop Behavior=::NewEnumerator1
    Loop Duration Mode=::NewEnumerator1
    UseLoopDelay=false
   )
   &SpawnBurst_Instantaneous(Spawn Count=8, Use Spawn Probability=false, UseLoopCountLimit=false)
  ]
  particleSpawnScript [
   &FN_SetSpawnGroup(Number of Spawn Groups=8)
   &InitializeParticle(
    Color=<rgba 1 0.6253 0.1354 0.45>
    Lifetime=2
    Mesh Scale Mode=::NewEnumerator3
    Sprite Size=&SelectVector2DFromArray(
     Direct Array Index=$Particles.SpawnGroup
     Vector 2D Selection Array=DataInterface() {propertyValues {FloatData [<2.025 3.996>, <2.565 4.4955>, <2.565 3.996>, <1.89 2.7972>, <2.025 2.22>, <2.565 2.96>, <2.565 2.96>, <1.89 2.22>]}}
    )
    Sprite Size Mode=::NewEnumerator4
    UsePositionOffset=true
    Write Material Random=true
   )
   &SpriteFacingAndAlignment(
    Alignment Coordinate Space=::Local
    Sprite Alignment=&SelectVectorFromArray(
     Direct Array Index=$Particles.SpawnGroup
     Vector Selection Array=DataInterface() {propertyValues {FloatData [<-0.2 1 0.7>, <0 1 0.7>, <0.2 1 0.7>, <0.4 1 0.7>, <-0.2 -1 0.7>, <0 -1 0.7>, <0.2 -1 0.7>, <0.4 -1 0.7>]}}
    )
    WriteAlignment=true
    WriteFacing=false
   )
   &CameraOffset(Camera Offset Amount=1)
  ]
  particleUpdateScript [
   &SkeletalMeshLocation(
    Bone / Socket Index=HlslExpression() {hlslExpression="(Particles.SpawnGroup>3) ? 1 : 0"}
    MaskSpawnGroup=false
    Mesh Sampling Type=::NewEnumerator5
    Orientation Sampling=::NewEnumerator1
    Sampled Position Offset=&SelectVectorFromArray(
     Direct Array Index=$Particles.SpawnGroup
     Vector Selection Array=DataInterface() {propertyValues {FloatData [<3.3 0 0>, <1.35 0 0>, <-0.7 0 0>, <-2.7 0 0>, <3.3 0 0>, <1.35 0 0>, <-0.7 0 0>, <-2.7 0 0>]}}
    )
    Skeletal Mesh=$User.SK_Mesh
    Socket Sampling Mode=::NewEnumerator5
   )
   &ParticleState(Kill Particles When Lifetime Has Elapsed=false, Loop Particles Lifetime=true)
   &ScaleSpriteSize(
    Scale Factor=&MakeVector2D(
     X=1
     Y=&FloatFromCurve(
      FloatCurve=DataInterface() {propertyValues {Curve {keys [<at 0 1 RCIM_Cubic>,<at 0.25 1.2 RCIM_Cubic>,<at 0.5 1 RCIM_Cubic>]}}}
      CurveIndex=&Add_Float(
       A=$Particles.NormalizedAge
       B=&Multiply_Float(A=&MakeFloatFromInt(Int=$Particles.SpawnGroup), B=-0.05)
      )
     )
    )
    Scale Sprite Size Mode=::NewEnumerator2
   )
   /SetVariables(Particles.TimeOfDayIntensity:NiagaraFloat=&Multiply_Float(
     A=0.8
     B=&FN_FrontEndFloatModifier(Gameplay=&FNTimeOfDayScalar(Night Scale=1, Morning Scale=1, Evening Scale=1), Front End=1)
    )) {isSetParameters=true, module=SetVariables_#1}
   /SetVariables(Particles.Color=&MultiplyLinearColor(
     Linear Color=$Particles.Initial.Color
     Scale Factor=&MakeVector4_FromFloat(Float=&Multiply_Float(A=$Particles.TimeOfDayIntensity, B=&FN_FrontEndFloatModifier(Gameplay=1, Front End=0.8)))
    )) {isSetParameters=true, module=SetVariables_#2}
  ]
  rendererClasses [&NiagaraSpriteRendererProperties]
  renderers [{rendererClass=&NiagaraSpriteRendererProperties}]
 }
 Front_Glow {
  emitterUpdateScript [
   &EmitterState(
    Inactive Response=::NewEnumerator1
    Life Cycle Mode=::NewEnumerator1
    Loop Behavior=::NewEnumerator1
    Loop Duration Mode=::NewEnumerator1
    UseLoopDelay=false
   )
   &SpawnBurst_Instantaneous(Use Spawn Probability=false, UseLoopCountLimit=false)
  ]
  particleSpawnScript [
   &InitializeParticle(
    Color=<rgba 1 0.6321 0.151 0.05>
    Mesh Scale Mode=::NewEnumerator3
    Position Offset=<0 50 -0.1>
    Sprite Size=<9.5 5.5>
    Sprite Size Mode=::NewEnumerator4
    UsePositionOffset=false
    Write Material Random=true
   )
   &SpriteFacingAndAlignment(Alignment Coordinate Space=::Local, Sprite Alignment=<0 0 1>, Sprite Facing=<-0.08 -1 0>, WriteAlignment=true, WriteFacing=true)
   &CameraOffset(Camera Offset Amount=10)
   /SetVariables(Particles.SpriteRotation=-15) {isSetParameters=true}
  ]
  particleUpdateScript [
   &SkeletalMeshLocation(
    MaskSpawnGroup=false
    Mesh Sampling Type=::NewEnumerator5
    Sampled Position Offset=<0.4 0 0>
    Skeletal Mesh=$User.SK_Mesh
    Socket Sampling Mode=::NewEnumerator5
   )
   &ParticleState(Kill Particles When Lifetime Has Elapsed=false)
   /SetVariables(Particles.TimeOfDayIntensity:NiagaraFloat=&Multiply_Float(
     A=1
     B=&FN_FrontEndFloatModifier(Gameplay=&FNTimeOfDayScalar(Night Scale=0.8, Morning Scale=1, Evening Scale=1), Front End=1)
    )) {isSetParameters=true}
   /SetVariables(Particles.Color=&MultiplyLinearColor(
     Linear Color=$Particles.Initial.Color
     Scale Factor=&MakeVector4_FromFloat(Float=$Particles.TimeOfDayIntensity)
    )) {isSetParameters=true}
   /SetVariables(
    Particles.AirGlowCameraOffset=10
    Particles.AirGlowColor:LinearColor=&MultiplyLinearColor(
     Linear Color=<rgba 1 0.7833 0.5 0.35>
     Scale Factor=&MakeVector4_FromFloat(Float=&Multiply_Float(A=$Particles.TimeOfDayIntensity, B=&FN_FrontEndFloatModifier(Gameplay=1, Front End=0.4)))
    )
    Particles.AirGlowSize:Vector2f=<25 13>
   ) {isSetParameters=true}
  ]
  rendererClasses [&NiagaraSpriteRendererProperties]
  renderers [
   {rendererClass=&NiagaraSpriteRendererProperties}
   {rendererClass=&NiagaraSpriteRendererProperties, rendererIndex=1}
  ]
 }
 Glow_Rays {
  emitterUpdateScript [
   &EmitterState(
    Inactive Response=::NewEnumerator1
    Life Cycle Mode=::NewEnumerator1
    Loop Behavior=::NewEnumerator1
    Loop Duration Mode=::NewEnumerator1
    UseLoopDelay=false
   )
   &SpawnBurst_Instantaneous(Use Spawn Probability=false, UseLoopCountLimit=false)
  ]
  particleSpawnScript [
   &FN_SetSpawnGroup()
   &InitializeParticle(Color=<rgba 1 0.5883 0.2031 1>, Mesh Scale Mode=::NewEnumerator3, UsePositionOffset=false, Write Material Random=true)
   &SystemLocation(MaskSpawnGroup=false, Offset=<0 -1 0>)
  ]
  particleUpdateScript [
   &ParticleState(Kill Particles When Lifetime Has Elapsed=false)
   &SolveForcesAndVelocity(Clamp Velocity=false, Limit Acceleration=false)
   /SetVariables(Particles.TimeOfDayIntensity:NiagaraFloat=&Multiply_Float(
     A=2.25
     B=&FN_FrontEndFloatModifier(Gameplay=&FNTimeOfDayScalar(Night Scale=0.6, Morning Scale=1, Evening Scale=0.7), Front End=1)
    )) {isSetParameters=true}
   /SetVariables(Particles.Color=&MultiplyLinearColor(
     Linear Color=$Particles.Initial.Color
     Scale Factor=&MakeVector4_FromFloat(Float=$Particles.TimeOfDayIntensity)
    )) {isSetParameters=true}
  ]
  rendererClasses [&NiagaraMeshRendererProperties]
  renderers [{rendererClass=&NiagaraMeshRendererProperties, rendererIndex=0}]
 }
 Glows {
  emitterUpdateScript [
   &EmitterState(
    Inactive Response=::NewEnumerator1
    Life Cycle Mode=::NewEnumerator1
    Loop Behavior=::NewEnumerator1
    Loop Duration Mode=::NewEnumerator1
    UseLoopDelay=false
   )
   &SpawnBurst_Instantaneous(Spawn Count=8, Use Spawn Probability=false, UseLoopCountLimit=false)
  ]
  particleSpawnScript [
   &FN_SetSpawnGroup(Number of Spawn Groups=8)
   &InitializeParticle(
    Color=<rgba 1 0.6773 0.2552 0.2>
    Lifetime=2
    Mesh Scale Mode=::NewEnumerator3
    Sprite Size Mode=::NewEnumerator3
    Uniform Sprite Size=5
    UsePositionOffset=true
    Write Material Random=true
   )
  ]
  particleUpdateScript [
   &SkeletalMeshLocation(
    Bone / Socket Index=HlslExpression() {hlslExpression="(Particles.SpawnGroup>3) ? 1 : 0"}
    MaskSpawnGroup=false
    Mesh Sampling Type=::NewEnumerator5
    Orientation Sampling=::NewEnumerator1
    Sampled Position Offset=&SelectVectorFromArray(
     Direct Array Index=$Particles.SpawnGroup
     Vector Selection Array=DataInterface() {propertyValues {FloatData [<3 -0.05 0>, <1.4 0 0>, <-0.6 0 0>, <-2.7 0 0>, <3 0 0>, <1.4 0 0>, <-0.6 0 0>, <-2.7 0 0>]}}
    )
    Skeletal Mesh=$User.SK_Mesh
    Socket Sampling Mode=::NewEnumerator5
   )
   &ParticleState(Kill Particles When Lifetime Has Elapsed=false, Loop Particles Lifetime=true)
   /SetVariables(Particles.TimeOfDayIntensity:NiagaraFloat=&Multiply_Float(
     A=0.8
     B=&FN_FrontEndFloatModifier(Gameplay=&FNTimeOfDayScalar(Night Scale=0.75, Morning Scale=1, Evening Scale=1), Front End=1)
    )) {isSetParameters=true, module=SetVariables_#1}
   /SetVariables(Particles.Color=&MultiplyLinearColor(
     Linear Color=$Particles.Initial.Color
     Scale Factor=&MakeVector4_FromFloat(Float=&Multiply_Float(A=$Particles.TimeOfDayIntensity, B=&FN_FrontEndFloatModifier(Gameplay=1, Front End=0.8)))
    )) {isSetParameters=true, module=SetVariables_#2}
   &CameraOffset(Camera Offset Amount=&MakeFloatFromInt(Int=HlslExpression() {hlslExpression="(Particles.SpawnGroup>3) ? 0 : 2"})) {module=CameraOffset001}
  ]
  rendererClasses [&NiagaraSpriteRendererProperties]
  renderers [{rendererClass=&NiagaraSpriteRendererProperties}]
 }
 InnerLights {
  emitterUpdateScript [
   &EmitterState(
    Inactive Response=::NewEnumerator1
    Life Cycle Mode=::NewEnumerator1
    Loop Behavior=::NewEnumerator1
    Loop Duration Mode=::NewEnumerator1
    UseLoopDelay=false
   )
   &SpawnBurst_Instantaneous(Use Spawn Probability=false, UseLoopCountLimit=false)
  ]
  particleSpawnScript [
   &FN_SetSpawnGroup()
   &InitializeParticle(Color=<rgba 15 9.651 2.6562 1>, Mesh Scale Mode=::NewEnumerator3, UsePositionOffset=false, Write Material Random=true)
   &SystemLocation(MaskSpawnGroup=false, Offset=<0 -1 0>)
  ]
  particleUpdateScript [
   &ParticleState(Kill Particles When Lifetime Has Elapsed=false)
   &SolveForcesAndVelocity(Clamp Velocity=false, Limit Acceleration=false)
   /SetVariables(Particles.TimeOfDayIntensity:NiagaraFloat=&Multiply_Float(
     A=1
     B=&FN_FrontEndFloatModifier(Gameplay=&FNTimeOfDayScalar(Night Scale=1, Morning Scale=1, Evening Scale=1), Front End=1)
    )) {isSetParameters=true}
   /SetVariables(Particles.Color=&MultiplyLinearColor(
     Linear Color=$Particles.Initial.Color
     Scale Factor=&MakeVector4_FromFloat(Float=$Particles.TimeOfDayIntensity)
    )) {isSetParameters=true}
  ]
  rendererClasses [&NiagaraMeshRendererProperties]
  renderers [
   {rendererClass=&NiagaraMeshRendererProperties, rendererIndex=0}
   {rendererClass=&NiagaraMeshRendererProperties, rendererIndex=1}
  ]
 }
}
summary {
 systemName=NS_Backpack_LastVoicePresent_Idle
 emitters [
  {emitterName=Beams, rendererClasses [&NiagaraSpriteRendererProperties]}
  {emitterName=Front_Glow, rendererClasses [&NiagaraSpriteRendererProperties]}
  {emitterName=Glows, rendererClasses [&NiagaraSpriteRendererProperties]}
  {emitterName=InnerLights, rendererClasses [&NiagaraMeshRendererProperties]}
  {emitterName=Glow_Rays, rendererClasses [&NiagaraMeshRendererProperties]}
 ]
 userVariables [
  User.bIsFrontEnd {type=NiagaraBool}
  User.bIsFrontEndPreview {type=NiagaraBool}
  User.SK_Mesh {
   type=NiagaraDataInterfaceSkeletalMesh
   defaultValue {
    struct=NiagaraExt_VariableValue_DataInterface
    value {dataInterface=&self:NiagaraDataInterfaceSkeletalMesh_0}
   }
  }
 ]
}
