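# Physics asset for the LastVoicePresent backpack cosmetic, setting up two dangling lace chains for physics simulation.
# Each six-capsule chain hangs from a compound sphere-and-capsule pack body through limited swing and twist joints.
# The left lace root swings widely, while the right root is angularly locked to the pack.
# Pelvis and spine capsules approximate the wearer's torso, each constrained to the pack with free angular motion.
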PhysicsAsset /BRCosmetics/Accessories/FORT_Backpacks/Backpack_M_MED_LastVoicePresent/Meshes/M_MED_LastVoicePresent_Pack_Physics
bodies {
 dyn_laces_l_1 [{shapeType=Capsule, length=0.505, radius=0.505, shapeName=dyn_laces_l_1_capsule, center=<0 0 -0.2642>, rotation=<rot 0 0 15>}]
 dyn_laces_l_2 [{shapeType=Capsule, length=2, radius=1, shapeName=dyn_laces_l_2_capsule, center=<0 -0.6207 -0.4173>, rotation=<rot 0 0 25>}]
 dyn_laces_l_3 [{shapeType=Capsule, length=5.1354, radius=1, shapeName=dyn_laces_l_3_capsule, center=<0 -0.8906 -0.7618>, rotation=<rot 0 0 5>}]
 dyn_laces_l_4 [{shapeType=Capsule, length=4, radius=1.25, shapeName=dyn_laces_l_4_capsule, center=<0 -0.5114 -1.2394>}]
 dyn_laces_l_5 [{shapeType=Capsule, length=2, radius=1.25, shapeName=dyn_laces_l_5_capsule, center=<0 0 -0.4717>}]
 dyn_laces_l_6 [{shapeType=Capsule, length=1.75, radius=1.25, shapeName=dyn_laces_l_6_capsule, center=<0 0 -0.7325>}]
 dyn_laces_r_1 [{shapeType=Capsule, length=0.2024, radius=0.8076, shapeName=dyn_laces_r_1_capsule}]
 dyn_laces_r_2 [{shapeType=Capsule, length=2, radius=1, shapeName=dyn_laces_r_2_capsule}]
 dyn_laces_r_3 [{shapeType=Capsule, length=0.75, radius=2.25, shapeName=dyn_laces_r_3_capsule, center=<0 0.4724 0>}]
 dyn_laces_r_4 [{shapeType=Capsule, length=2.25, radius=2.5, shapeName=dyn_laces_r_4_capsule, center=<0 0.0851 -0.7984>}]
 dyn_laces_r_5 [{shapeType=Capsule, length=4, radius=2.5, shapeName=dyn_laces_r_5_capsule}]
 dyn_laces_r_6 [{shapeType=Capsule, length=0.5, radius=2.75, shapeName=dyn_laces_r_6_capsule}]
 pack [
  {shapeType=Sphere, radius=6.75, shapeName=pack_sphere, center=<-9.563 -7.0809 6.4067>}
  {
   shapeType=Capsule
   length=16.2759
   radius=5.6887
   shapeName=pack_capsule
   center=<-1.8903 -6.6489 -3.0132>
   rotation=<rot 75 -0.0017 -0.0018>
  }
  {shapeType=Capsule, length=3.5259, radius=3, shapeName=pack_capsule, center=<0.4242 -5.2917 4.1261>, rotation=<rot 35 -0.0004 -0.0007>}
  {shapeType=Capsule, length=21, radius=9, shapeName=pack_capsule, center=<0.0043 -0.8796 7.0266>, rotation=<rot 70 -0.0001 -0.0001>}
 ]
 pelvis [{shapeType=Capsule, length=12, radius=8, shapeName=pelvis_capsule}]
 spine_01 [{shapeType=Capsule, length=12, radius=8.25, shapeName=spine_01_capsule}]
 spine_02 [{shapeType=Capsule, length=10, radius=10.25, shapeName=spine_02_capsule}]
 spine_03 [{shapeType=Capsule, length=9.25, radius=11, shapeName=spine_03_capsule}]
}
constraints [
 {bone1Name=dyn_laces_r_2, bone2Name=dyn_laces_r_1, swing1LimitDegrees=20, swing2LimitDegrees=20, twistLimitDegrees=15}
 {bone1Name=dyn_laces_r_3, bone2Name=dyn_laces_r_2, swing1LimitDegrees=30, swing2LimitDegrees=30, twistLimitDegrees=20}
 {bone1Name=dyn_laces_r_4, bone2Name=dyn_laces_r_3, swing1LimitDegrees=30, swing2LimitDegrees=30, twistLimitDegrees=20}
 {bone1Name=dyn_laces_r_6, bone2Name=dyn_laces_r_5, swing1LimitDegrees=30, swing2LimitDegrees=30, twistLimitDegrees=20}
 {bone1Name=dyn_laces_l_2, bone2Name=dyn_laces_l_1, swing1LimitDegrees=170, swing2LimitDegrees=160, twistLimitDegrees=20}
 {bone1Name=dyn_laces_l_3, bone2Name=dyn_laces_l_2, swing1LimitDegrees=30, swing2LimitDegrees=30, twistLimitDegrees=20}
 {bone1Name=dyn_laces_l_4, bone2Name=dyn_laces_l_3, swing1LimitDegrees=30, swing2LimitDegrees=30, twistLimitDegrees=20}
 {bone1Name=dyn_laces_l_6, bone2Name=dyn_laces_l_5, swing1LimitDegrees=30, swing2LimitDegrees=30, twistLimitDegrees=20}
 {bone1Name=dyn_laces_l_5, bone2Name=dyn_laces_l_4, swing1LimitDegrees=30, swing2LimitDegrees=30, twistLimitDegrees=20}
 {bone1Name=dyn_laces_r_5, bone2Name=dyn_laces_r_4, swing1LimitDegrees=30, swing2LimitDegrees=30, twistLimitDegrees=20}
 {bone1Name=dyn_laces_l_1, bone2Name=pack, swing1LimitDegrees=150, swing2LimitDegrees=90, twistLimitDegrees=10}
 {
  bone1Name=dyn_laces_r_1
  bone2Name=pack
  swing1LimitDegrees=45
  swing1Motion=Locked
  swing2LimitDegrees=45
  swing2Motion=Locked
  twistLimitDegrees=45
  twistMotion=Locked
 }
 {
  bone1Name=pelvis
  bone2Name=pack
  swing1LimitDegrees=45
  swing1Motion=Free
  swing2LimitDegrees=45
  swing2Motion=Free
  twistLimitDegrees=45
  twistMotion=Free
 }
 {
  bone1Name=spine_01
  bone2Name=pack
  swing1LimitDegrees=45
  swing1Motion=Free
  swing2LimitDegrees=45
  swing2Motion=Free
  twistLimitDegrees=45
  twistMotion=Free
 }
 {
  bone1Name=spine_02
  bone2Name=pack
  swing1LimitDegrees=45
  swing1Motion=Free
  swing2LimitDegrees=45
  swing2Motion=Free
  twistLimitDegrees=45
  twistMotion=Free
 }
 {
  bone1Name=spine_03
  bone2Name=pack
  swing1LimitDegrees=45
  swing1Motion=Free
  swing2LimitDegrees=45
  swing2Motion=Free
  twistLimitDegrees=45
  twistMotion=Free
 }
]
