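# Animation Blueprint for the LastVoicePresent backpack, adding physics motion while following the character's body.
# Each update copies body pelvis and spine transforms, derives smoothed gravity from movement state, and sets gameplay inertia scales.
# Two similar branches pin those bones, offset right laces (more when skydiving), then run RigidBody; bIsInFrontEnd selects between them.
# The AnimDynamics lace chains are not connected to the output.
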AnimBlueprint /BRCosmetics/Accessories/FORT_Backpacks/Backpack_M_MED_LastVoicePresent/Meshes/M_MED_LastVoicePresent_Pack_AnimBP
target_skeleton=&/M_MED_LastVoicePresent_Pack_Skeleton
graphs [
 AnimGraph {nodes [
   n1 = "Anim Dynamics - Chain: dyn_laces_r_2 -> dyn_laces_r_6"(Alpha=1, ExternalForce="0,0,0") {
    angularDampingOverride=0.98
    angularSpringConstant=5
    bChain=true
    bLinearSpring=true
    bOverrideAngularDamping=true
    bUsePlanarLimit=false
    gravityScale=0.5
    linearSpringConstant=400
    numSolverIterationsPostUpdate=2
    numSolverIterationsPreUpdate=8
    alphaBoolBlend {blendOption=Linear}
    boundBone {boneName=dyn_laces_r_2}
    chainEnd {boneName=dyn_laces_r_6}
    componentAppliedLinearAccClamp=<5000 5000 5000>
    componentLinearAccScale=<0.5 0.5 0.5>
    componentLinearVelScale=<0.5 0.5 0.5>
    physicsBodyDefinitions [
     {
      boundBone {boneName=dyn_laces_r_2}
      boxExtents=<20 20 1>
      constraintSetup {angularLimitsMax=<10 10 5>, angularLimitsMin=<-10 -10 -5>, angularTarget=<0 0 0>}
      localJointOffset=<0 0 -2>
     }
     {
      boundBone {boneName=dyn_laces_r_3}
      boxExtents=<20 20 1>
      constraintSetup {angularLimitsMax=<10 10 5>, angularLimitsMin=<-10 -10 -5>, angularTarget=<0 0 0>}
      localJointOffset=<0 0 -2>
     }
     {
      boundBone {boneName=dyn_laces_r_4}
      boxExtents=<20 20 1>
      constraintSetup {angularLimitsMax=<10 10 5>, angularLimitsMin=<-10 -10 -5>, angularTarget=<0 0 0>}
      localJointOffset=<0 0 -2>
     }
     {
      boundBone {boneName=dyn_laces_r_5}
      boxExtents=<20 20 1>
      constraintSetup {angularLimitsMax=<10 10 5>, angularLimitsMin=<-10 -10 -5>, angularTarget=<0 0 0>}
      localJointOffset=<0 0 -2>
     }
     {
      boundBone {boneName=dyn_laces_r_6}
      boxExtents=<20 20 1>
      constraintSetup {angularLimitsMax=<10 10 5>, angularLimitsMin=<-10 -10 -5>, angularTarget=<0 0 0>}
      localJointOffset=<0 0 -2>
     }
    ]
   }
   "Anim Dynamics - Chain: dyn_laces_l_2 -> dyn_laces_l_6"(ComponentPose=@n1, Alpha=1, ExternalForce="0,0,0") {
    angularDampingOverride=0.9
    angularSpringConstant=100
    bAngularSpring=true
    bChain=true
    bLinearSpring=true
    bOverrideLinearDamping=true
    bUsePlanarLimit=false
    gravityScale=0.5
    linearDampingOverride=0.98
    linearSpringConstant=500
    numSolverIterationsPostUpdate=2
    numSolverIterationsPreUpdate=8
    simulationSpace=Actor
    alphaBoolBlend {blendOption=Linear}
    boundBone {boneName=dyn_laces_l_2}
    chainEnd {boneName=dyn_laces_l_6}
    componentAppliedLinearAccClamp=<5000 5000 5000>
    componentLinearAccScale=<0.5 0.5 0.5>
    componentLinearVelScale=<0.5 0.5 0.5>
    physicsBodyDefinitions [
     {
      boundBone {boneName=dyn_laces_l_2}
      boxExtents=<20 20 1>
      constraintSetup {angularLimitsMax=<10 10 5>, angularLimitsMin=<-10 10 -5>, angularTarget=<0 0 0>}
      localJointOffset=<0 0 -2>
     }
     {
      boundBone {boneName=dyn_laces_l_3}
      boxExtents=<20 20 1>
      constraintSetup {angularLimitsMax=<0 0 0>, angularLimitsMin=<0 0 0>, angularTarget=<0 0 0>}
      localJointOffset=<0 0 -2>
     }
     {
      boundBone {boneName=dyn_laces_l_4}
      boxExtents=<20 20 1>
      constraintSetup {angularLimitsMax=<10 10 5>, angularLimitsMin=<-10 -10 -5>, angularTarget=<0 0 0>}
      localJointOffset=<0 0 -2>
     }
     {
      boundBone {boneName=dyn_laces_l_5}
      boxExtents=<20 20 1>
      constraintSetup {angularLimitsMax=<10 10 5>, angularLimitsMin=<-10 -10 -5>, angularTarget=<0 0 0>}
      localJointOffset=<0 0 -2>
     }
     {
      boundBone {boneName=dyn_laces_l_6}
      boxExtents=<20 20 1>
      constraintSetup {angularLimitsMax=<10 10 5>, angularLimitsMin=<-10 -10 -5>, angularTarget=<0 0 0>}
      localJointOffset=<0 0 -2>
     }
    ]
   }
   n2 = "Anim Dynamics - Chain: dyn_laces_r_2 -> dyn_laces_r_6"(Alpha=1, ExternalForce="0,0,0") {
    angularDampingOverride=0.98
    angularSpringConstant=5
    bChain=true
    bLinearSpring=true
    bOverrideAngularDamping=true
    bUsePlanarLimit=false
    gravityScale=0.5
    linearSpringConstant=400
    numSolverIterationsPostUpdate=2
    numSolverIterationsPreUpdate=8
    alphaBoolBlend {blendOption=Linear}
    boundBone {boneName=dyn_laces_r_2}
    chainEnd {boneName=dyn_laces_r_6}
    componentAppliedLinearAccClamp=<5000 5000 5000>
    componentLinearAccScale=<0.5 0.5 0.5>
    componentLinearVelScale=<0.5 0.5 0.5>
    physicsBodyDefinitions [
     {
      boundBone {boneName=dyn_laces_r_2}
      boxExtents=<20 20 1>
      constraintSetup {angularLimitsMax=<10 10 5>, angularLimitsMin=<-10 -10 -5>, angularTarget=<0 0 0>}
      localJointOffset=<0 0 -2>
     }
     {
      boundBone {boneName=dyn_laces_r_3}
      boxExtents=<20 20 1>
      constraintSetup {angularLimitsMax=<10 10 5>, angularLimitsMin=<-10 -10 -5>, angularTarget=<0 0 0>}
      localJointOffset=<0 0 -2>
     }
     {
      boundBone {boneName=dyn_laces_r_4}
      boxExtents=<20 20 1>
      constraintSetup {angularLimitsMax=<10 10 5>, angularLimitsMin=<-10 -10 -5>, angularTarget=<0 0 0>}
      localJointOffset=<0 0 -2>
     }
     {
      boundBone {boneName=dyn_laces_r_5}
      boxExtents=<20 20 1>
      constraintSetup {angularLimitsMax=<10 10 5>, angularLimitsMin=<-10 -10 -5>, angularTarget=<0 0 0>}
      localJointOffset=<0 0 -2>
     }
     {
      boundBone {boneName=dyn_laces_r_6}
      boxExtents=<20 20 1>
      constraintSetup {angularLimitsMax=<10 10 5>, angularLimitsMin=<-10 -10 -5>, angularTarget=<0 0 0>}
      localJointOffset=<0 0 -2>
     }
    ]
   }
   "Anim Dynamics - Chain: dyn_laces_l_2 -> dyn_laces_l_6"(ComponentPose=@n2, Alpha=1, ExternalForce="0,0,0") {
    angularDampingOverride=0.9
    angularSpringConstant=100
    bAngularSpring=true
    bChain=true
    bLinearSpring=true
    bOverrideLinearDamping=true
    bUsePlanarLimit=false
    gravityScale=0.5
    linearDampingOverride=0.98
    linearSpringConstant=500
    numSolverIterationsPostUpdate=2
    numSolverIterationsPreUpdate=8
    simulationSpace=Actor
    alphaBoolBlend {blendOption=Linear}
    boundBone {boneName=dyn_laces_l_2}
    chainEnd {boneName=dyn_laces_l_6}
    componentAppliedLinearAccClamp=<5000 5000 5000>
    componentLinearAccScale=<0.5 0.5 0.5>
    componentLinearVelScale=<0.5 0.5 0.5>
    physicsBodyDefinitions [
     {
      boundBone {boneName=dyn_laces_l_2}
      boxExtents=<20 20 1>
      constraintSetup {angularLimitsMax=<10 10 5>, angularLimitsMin=<-10 10 -5>, angularTarget=<0 0 0>}
      localJointOffset=<0 0 -2>
     }
     {
      boundBone {boneName=dyn_laces_l_3}
      boxExtents=<20 20 1>
      constraintSetup {angularLimitsMax=<0 0 0>, angularLimitsMin=<0 0 0>, angularTarget=<0 0 0>}
      localJointOffset=<0 0 -2>
     }
     {
      boundBone {boneName=dyn_laces_l_4}
      boxExtents=<20 20 1>
      constraintSetup {angularLimitsMax=<10 10 5>, angularLimitsMin=<-10 -10 -5>, angularTarget=<0 0 0>}
      localJointOffset=<0 0 -2>
     }
     {
      boundBone {boneName=dyn_laces_l_5}
      boxExtents=<20 20 1>
      constraintSetup {angularLimitsMax=<10 10 5>, angularLimitsMin=<-10 -10 -5>, angularTarget=<0 0 0>}
      localJointOffset=<0 0 -2>
     }
     {
      boundBone {boneName=dyn_laces_l_6}
      boxExtents=<20 20 1>
      constraintSetup {angularLimitsMax=<10 10 5>, angularLimitsMin=<-10 -10 -5>, angularTarget=<0 0 0>}
      localJointOffset=<0 0 -2>
     }
    ]
   }
   n3 = "Transform (Modify) Bone - Bone: pelvis"(
    ComponentPose=Local To Component(LocalPose=Input Pose())
    Alpha=1
    Translation=Get pelvis_location() -> (pelvis_location:Vector)
    Rotation=Get pelvis_rotation() -> (pelvis_rotation:Rotator)
    Scale="1,1,1"
   ) {
    rotationMode=BMM_Replace
    rotationSpace=BCS_WorldSpace
    translationMode=BMM_Replace
    translationSpace=BCS_WorldSpace
    alphaBoolBlend {blendOption=Linear}
    boneToModify {boneName=pelvis}
   }
   n4 = "Transform (Modify) Bone - Bone: spine_01"(
    ComponentPose=@n3
    Alpha=1
    Translation=Get spine_01_location() -> (spine_01_location:Vector)
    Rotation=Get spine_01_rotation() -> (spine_01_rotation:Rotator)
    Scale="1,1,1"
   ) {
    rotationMode=BMM_Replace
    rotationSpace=BCS_WorldSpace
    translationMode=BMM_Replace
    translationSpace=BCS_WorldSpace
    alphaBoolBlend {blendOption=Linear}
    boneToModify {boneName=spine_01}
   }
   n5 = "Transform (Modify) Bone - Bone: spine_02"(
    ComponentPose=@n4
    Alpha=1
    Translation=Get spine_02_location() -> (spine_02_location:Vector)
    Rotation=Get spine_02_rotation() -> (spine_02_rotation:Rotator)
    Scale="1,1,1"
   ) {
    rotationMode=BMM_Replace
    rotationSpace=BCS_WorldSpace
    translationMode=BMM_Replace
    translationSpace=BCS_WorldSpace
    alphaBoolBlend {blendOption=Linear}
    boneToModify {boneName=spine_02}
   }
   n6 = "Transform (Modify) Bone - Bone: spine_03"(
    ComponentPose=@n5
    Alpha=1
    Translation=Get spine_03_location() -> (spine_03_location:Vector)
    Rotation=Get spine_03_rotation() -> (spine_03_rotation:Rotator)
    Scale="1,1,1"
   ) {
    rotationMode=BMM_Replace
    rotationSpace=BCS_WorldSpace
    translationMode=BMM_Replace
    translationSpace=BCS_WorldSpace
    alphaBoolBlend {blendOption=Linear}
    boneToModify {boneName=spine_03}
   }
   n7 = "Transform (Modify) Bone - Bone: dyn_laces_r_4"(ComponentPose=@n6, Alpha=1, Translation="6,0,0", Rotation="0,0,0", Scale="1,1,1") {
    rotationMode=BMM_Additive
    translationMode=BMM_Additive
    alphaBoolBlend {blendInTime=0.3, blendOption=ExpInOut, blendOutTime=0.3}
    boneToModify {boneName=dyn_laces_r_4}
    translation {x=6}
   }
   n8 = Get bIsSkyDiving()
   n9 = "Transform (Modify) Bone - Bone: dyn_laces_r_4"!(ComponentPose=@n7, bAlphaBoolEnabled:bool=@n8, Translation="0,0,0", Rotation="0,0,15", Scale="1,1,1") -> (Pose) {
    alphaInputType=Bool
    rotationMode=BMM_Additive
    alphaBoolBlend {blendInTime=0.3, blendOption=ExpInOut, blendOutTime=0.3}
    boneToModify {boneName=dyn_laces_r_4}
    rotation {roll=15}
   }
   n10 = "Transform (Modify) Bone - Bone: dyn_laces_r_5"!(ComponentPose=@n9, bAlphaBoolEnabled:bool=@n8, Translation="0,0,0", Rotation="0,0,5", Scale="1,1,1") -> (Pose) {
    alphaInputType=Bool
    rotationMode=BMM_Additive
    alphaBoolBlend {blendInTime=0.3, blendOption=ExpInOut, blendOutTime=0.3}
    boneToModify {boneName=dyn_laces_r_5}
    rotation {roll=5}
   }
   n11 = RigidBody!(
    ComponentPose=@n10
    bAlphaBoolEnabled:bool
    OverrideWorldGravity:Vector=Get Gravity Override()
    ExternalForce
    ComponentLinearAccScale:Vector="0.5,0.5,0.5"
    ComponentLinearVelScale:Vector="0.5,0.5,0.5"
   ) -> (Pose) {
    alphaInputType=Bool
    bClampLinearTranslationLimitToRefPose=true
    bOverrideWorldGravity=true
    overridePhysicsAsset=&/M_MED_LastVoicePresent_Pack_Physics
    simulationSpace=BaseBoneSpace
    alphaBoolBlend {blendInTime=0.5, blendOption=Linear, blendOutTime=0.5}
    baseBoneRef {boneName=root}
    componentAppliedLinearAccClamp=<2000 2000 2000>
    componentLinearAccScale=<0.3 0.3 0.3>
    componentLinearVelScale=<0.3 0.3 0.3>
    overrideWorldGravity {z=-980}
   }
   n12 = "Transform (Modify) Bone - Bone: root"(ComponentPose=@n11, Translation="0,-1,0", Rotation="0,0,0") {
    translationMode=BMM_Additive
    translationSpace=BCS_BoneSpace
    alphaBoolBlend {blendOption=Linear}
    boneToModify {boneName=root}
    translation {y=-1}
   }
   n13 = "Transform (Modify) Bone - Bone: pelvis"(ComponentPose=Local To Component(), Alpha=1, Translation=Get pelvis_location(), Rotation=Get pelvis_rotation(), Scale="1,1,1") {
    rotationMode=BMM_Replace
    rotationSpace=BCS_WorldSpace
    translationMode=BMM_Replace
    translationSpace=BCS_WorldSpace
    alphaBoolBlend {blendOption=Linear}
    boneToModify {boneName=pelvis}
   }
   n14 = "Transform (Modify) Bone - Bone: spine_01"(ComponentPose=@n13, Alpha=1, Translation=Get spine_01_location(), Rotation=Get spine_01_rotation(), Scale="1,1,1") {
    rotationMode=BMM_Replace
    rotationSpace=BCS_WorldSpace
    translationMode=BMM_Replace
    translationSpace=BCS_WorldSpace
    alphaBoolBlend {blendOption=Linear}
    boneToModify {boneName=spine_01}
   }
   n15 = "Transform (Modify) Bone - Bone: spine_02"(ComponentPose=@n14, Alpha=1, Translation=Get spine_02_location(), Rotation=Get spine_02_rotation(), Scale="1,1,1") {
    rotationMode=BMM_Replace
    rotationSpace=BCS_WorldSpace
    translationMode=BMM_Replace
    translationSpace=BCS_WorldSpace
    alphaBoolBlend {blendOption=Linear}
    boneToModify {boneName=spine_02}
   }
   n16 = "Transform (Modify) Bone - Bone: spine_03"(ComponentPose=@n15, Alpha=1, Translation=Get spine_03_location(), Rotation=Get spine_03_rotation(), Scale="1,1,1") {
    rotationMode=BMM_Replace
    rotationSpace=BCS_WorldSpace
    translationMode=BMM_Replace
    translationSpace=BCS_WorldSpace
    alphaBoolBlend {blendOption=Linear}
    boneToModify {boneName=spine_03}
   }
   n17 = "Transform (Modify) Bone - Bone: dyn_laces_r_4"!(ComponentPose=@n16, Alpha:float=1, Translation="6,0,0", Rotation="0,0,0", Scale="1,1,1") -> (Pose) {
    rotationMode=BMM_Additive
    translationMode=BMM_Additive
    alphaBoolBlend {blendInTime=0.3, blendOption=ExpInOut, blendOutTime=0.3}
    boneToModify {boneName=dyn_laces_r_4}
    translation {x=6}
   }
   n18 = Get bIsSkyDiving()
   n19 = "Transform (Modify) Bone - Bone: dyn_laces_r_4"!(ComponentPose=@n17, bAlphaBoolEnabled:bool=@n18, Translation="0,0,0", Rotation="0,0,15", Scale="1,1,1") -> (Pose) {
    alphaInputType=Bool
    rotationMode=BMM_Additive
    alphaBoolBlend {blendInTime=0.3, blendOption=ExpInOut, blendOutTime=0.3}
    boneToModify {boneName=dyn_laces_r_4}
    rotation {roll=15}
   }
   n20 = "Transform (Modify) Bone - Bone: dyn_laces_r_5"(ComponentPose=@n19, bAlphaBoolEnabled=@n18, Translation="0,0,0", Rotation="0,0,5", Scale="1,1,1") {
    alphaInputType=Bool
    rotationMode=BMM_Additive
    alphaBoolBlend {blendInTime=0.3, blendOption=ExpInOut, blendOutTime=0.3}
    boneToModify {boneName=dyn_laces_r_5}
    rotation {roll=5}
   }
   n21 = RigidBody(
    ComponentPose=@n20
    OverrideWorldGravity=Get Gravity Override()
    ComponentLinearAccScale=Get Component Linear Acc Scale() -> (Component Linear Acc Scale:Vector)
    ComponentLinearVelScale=Get Component Linear Vel Scale() -> (Component Linear Vel Scale:Vector)
   ) {
    alphaInputType=Bool
    bClampLinearTranslationLimitToRefPose=true
    bOverrideWorldGravity=true
    overridePhysicsAsset=&M_MED_LastVoicePresent_Pack_Physics
    simulationSpace=BaseBoneSpace
    alphaBoolBlend {blendInTime=0.5, blendOption=Linear, blendOutTime=0.5}
    baseBoneRef {boneName=root}
    componentAppliedLinearAccClamp=<5000 5000 5000>
    componentLinearAccScale=<0.3 0.3 0.3>
    componentLinearVelScale=<0.3 0.3 0.3>
    overrideWorldGravity {z=-980}
   }
   n22 = "Transform (Modify) Bone - Bone: root"(ComponentPose=@n21, Translation="0,-1,0", Rotation="0,0,0") {
    translationMode=BMM_Additive
    translationSpace=BCS_BoneSpace
    alphaBoolBlend {blendOption=Linear}
    boneToModify {boneName=root}
    translation {y=-1}
   }
   Output Pose(Result=Blend Poses by bool(BlendPose_0=Component To Local(ComponentPose=@n12), BlendPose_1=Component To Local(ComponentPose=@n22), bActiveValue=Get bIsInFrontEnd()) {blendType=Linear}) {layerGroup=None}
  ]}
 EventGraph {nodes [
   n1 = Event BlueprintUpdateAnimation()
   > Set Delta Time X(Delta Time X:double=@n1.DeltaTimeX) -> (Output_Get:double)
   > Is Valid(InputObject=Get Body Mesh() -> (Body Mesh:Skeletal Mesh Component Object Reference))
   > n2 = Gravity Override State Function DEPRECATED(CharacterPartAnimInstance=Self-Reference(), GameTimeInSeconds=GetGameTimeInSeconds(), GravityOverrideParams=Get Gravity Override Params())
   n3 = SelectInt(A=1, B=0, bPickA=AND Boolean(A=Get bIsBackpedaling() -> (bIsBackpedaling:bool), B=Get bIsMoving2D()))
   n4 = Get bIsCrouching()
   n5 = NOT Boolean(A=Get bIsSkyDiving())
   n6 = OR Boolean(A=Get bIsJumping(), B=Get bIsFalling())
   n7 = SelectInt(
    A=4
    B=SelectInt(A=3, B=SelectInt(A=2, B=@n3, bPickA=@n4), bPickA=AND Boolean(A=@n6, B=@n5))
    bPickA=Get bIsWaterJump() -> (bIsWaterJump:bool)
   )
   n8 = BreakVector(InVec=@n2)
   n9 = BreakVector(InVec=@n2)
   n10 = Make Array!(
    "[0]":Vector=@n2
    "[1]":Vector=MakeVector(X=@n9, Y=@n9.Y, Z=-150)
    "[2]":Vector=MakeVector(X=@n8, Y=@n8.Y, Z=-300)
    "[3]":Vector="0,0,-500"
    "[4]":Vector="0,0,-980"
   ) -> (Array:Array of Vectors)
   > n11 = Set Gravity Override Interp(Gravity Override Interp:Vector=VInterpTo!(
     Current:Vector=Get Gravity Override Interp() -> (Gravity Override Interp:Vector)
     Target:Vector="Get (a ref)"(Array:Array of Vectors=@n10, Dimension 1=@n7) -> (Output:"Vector (by ref)")
     DeltaTime:float=Get Delta Time X() -> (Delta Time X:double)
     InterpSpeed:float=8
    ) -> (ReturnValue:Vector)) -> (Output_Get:Vector)
   > Set Gravity Override(Gravity Override=@n11)
   n12 = SelectVector(A="0.1,0.35,0.35", B="1,1,0.6", bPickA=OR Boolean(A=Get bIsSkyDiving(), B=Get bIsWaterJump()))
   > Set Component Linear Acc Scale(Component Linear Acc Scale:Vector=@n12) -> (Output_Get:Vector)
   > Set Component Linear Vel Scale(Component Linear Vel Scale:Vector=@n12) -> (Output_Get:Vector)
   n13 = Get Body Mesh()
   n14 = BreakTransform(InTransform=GetSocketTransform(self=@n13, InSocketName=spine_01, TransformSpace=RTS_World))
   > Set spine_01_location(spine_01_location:Vector=@n14) -> (Output_Get:Vector)
   > Set spine_01_rotation(spine_01_rotation:Rotator=@n14.Rotation) -> (Output_Get:Rotator)
   n15 = BreakTransform(InTransform=GetSocketTransform(self=@n13, InSocketName=spine_02, TransformSpace=RTS_World))
   > Set spine_02_location(spine_02_location:Vector=@n15) -> (Output_Get:Vector)
   > Set spine_02_rotation(spine_02_rotation:Rotator=@n15.Rotation) -> (Output_Get:Rotator)
   n16 = BreakTransform(InTransform=GetSocketTransform(self=@n13, InSocketName=pelvis, TransformSpace=RTS_World))
   > Set pelvis_location(pelvis_location:Vector=@n16) -> (Output_Get:Vector)
   > Set pelvis_rotation(pelvis_rotation:Rotator=@n16.Rotation) -> (Output_Get:Rotator)
   n17 = BreakTransform(InTransform=GetSocketTransform(self=@n13, InSocketName=spine_03, TransformSpace=RTS_World))
   > Set spine_03_location(spine_03_location:Vector=@n17) -> (Output_Get:Vector)
   > Set spine_03_rotation(spine_03_rotation:Rotator=@n17.Rotation) -> (Output_Get:Rotator)
   n18 = Event OnCharacterCustomizationLoaded()
   > n19 = Sequence()
   > n20 = Cast To PlayerPawn_Athena(Object=TryGetPawnOwner())
   > Set Body Mesh(Body Mesh:Skeletal Mesh Component Object Reference=GetSkeletalMeshForPartType(self=@n20)) -> (Output_Get:Skeletal Mesh Component Object Reference)
   Set Body Mesh(execute=@n19.then_1, Body Mesh=@n18.CustomBaseMesh)
  ]}
]
variables [
 Body Mesh {type=SkeletalMeshComponent}
 Component Linear Acc Scale {type=Vector}
 Component Linear Vel Scale {type=Vector}
 Delta Time X {type=number}
 Gravity Override {type=Vector}
 Gravity Override Interp {type=Vector}
 Gravity Override Params {
  type=GravityOverrideParamsStruct
  default {
   onGround_Falling {gravityOverride=<0 0 -400>}
   onGround_Moving {gravityOverride=<50 0 -10>}
   skydive_DivingDown {gravityOverride=<0 0 -100>, windFrequency=<1.5 16 16>}
   skydive_FloatingDown {gravityOverride=<0 0 100>, windAmplitude=<200 600 200>, windFrequency=<2 3 3>, windOffset=<0 500 400>}
   skydive_GoingUp {gravityOverride=<0 0 -100>, windFrequency=<1 1 2>}
   skydive_Parachuting {gravityOverride=<0 0 100>, windAmplitude=<200 600 200>, windFrequency=<2 3 3>, windOffset=<0 500 400>}
   tactical_Sprint {gravityOverride=<0 0 0>}
   ziplining {gravityOverride=<0 0 -50>}
  }
 }
 pelvis_location {type=Vector}
 pelvis_rotation {type=Rotator}
 spine_01_location {type=Vector}
 spine_01_rotation {type=Rotator}
 spine_02_location {type=Vector}
 spine_02_rotation {type=Rotator}
 spine_03_location {type=Vector}
 spine_03_rotation {type=Rotator}
]
